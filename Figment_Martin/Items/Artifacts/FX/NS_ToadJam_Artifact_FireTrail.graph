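# Niagara fire trail effect for the ToadJam artifact, with a glow and light, flash flares, a ribbon trail and sprite smoke.
# Ribbon points are placed along the owner's local X axis and lerp toward those goals with increasing lag down the trail.
# A system camera-proximity value scales sizes and intensity, while the timeline drives fading and User.FadeOut stops smoke spawning.
# Several duplicate glow, trail and flare emitters are disabled.
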NiagaraSystem /Figment_Martin/Items/Artifacts/FX/NS_ToadJam_Artifact_FireTrail
dependencies {
 usedDataInterfaces [
  &NiagaraDataInterfaceCurve
  &NiagaraDataInterfaceParticleRead
  &NiagaraDataInterfaceColorCurve
  &NiagaraDataInterfaceArrayFloat
  &NiagaraDataInterfaceCamera
 ]
 usedDynamicInputs [
  &Cosine
  &ClampFloat
  &NormalizeFloat
  &DistanceBetweenPositions
  &ReturnCameraPosition
 ]
 usedModules [
  &SystemState
  &Timeline
  &self:SystemSpawnScript.SystemScriptSource.SystemScriptGraph.NiagaraNodeAssignment_0.SetVariables_#1
  &self:SystemSpawnScript.SystemScriptSource.SystemScriptGraph.NiagaraNodeAssignment_1.SetVariables_#2
  &self:SystemSpawnScript.SystemScriptSource.SystemScriptGraph.NiagaraNodeAssignment_2.SetVariables_#3
 ]
}
emitters {
 Glow_and_Light {
  emitterUpdateScript [
   &EmitterState(
    Inactive Response=::NewEnumerator1
    Life Cycle Mode=::NewEnumerator1
    Loop Behavior=::NewEnumerator1
    Loop Duration Mode=::NewEnumerator1
    UseLoopDelay=false
   )
   &SpawnBurst_Instantaneous(Spawn Count=1, Use Spawn Probability=false)
   /SetVariables(Emitter.BurstAlpha:NiagaraFloat=&FloatFromCurve(
     FloatCurve=DataInterface() {propertyValues {Curve {keys [
         {arriveTangent=5.6877, interpMode=RCIM_Cubic, leaveTangent=5.6877, tangentMode=RCTM_User, time=0, value=0}
         {arriveTangent=-3.7229, interpMode=RCIM_Cubic, leaveTangent=-3.7229, tangentMode=RCTM_User, time=0.1, value=1}
         {interpMode=RCIM_Cubic, tangentMode=RCTM_User, time=0.5, value=0}
        ]}}}
     CurveIndex=$Emitter.Age
    )) {isSetParameters=true, module=SetVariables_#4}
  ]
  particleSpawnScript [
   &InitializeParticle(
    Color=<rgba 5 1 0 2>
    Lifetime=0.5
    Mesh Scale=<1.2 1 3>
    Mesh Scale Mode=::NewEnumerator4
    Position Offset=<-100 0 0>
    Sprite Size Mode=::NewEnumerator3
    Uniform Sprite Size=300
    UsePositionOffset=true
    Write Material Random=true
   ) {module=InitializeParticle001}
   /SetVariables(Particles.LightExponent=2.5) {isSetParameters=true, module=SetVariables_#5}
   /SetVariables(Particles.MeshOrientation=&MakeQuatFromAxisAngle(Axis=<0 1 0>, AngleInDegrees=90)) {isSetParameters=true, module=SetVariables_#6}
   /SetVariables(Particles.GlowDirection:Vector3f=&TransformVector(Vector=<1 0 0>, Destination Space=::World)) {isSetParameters=true, module=SetVariables_#7}
   &CameraOffset(Camera Offset Amount=30)
   &SystemLocation(MaskSpawnGroup=false, Offset=<30 0 0>)
  ]
  particleUpdateScript [
   /SetVariables(Particles.LightColor=&MultiplyLinearColor(
     Linear Color=<rgba 10 0.6667 0 1>
     Scale Factor=&MakeVector4_FromFloat(Float=&Multiply_Float(
       A=&Sine(Normalized Angle=$Emitter.Age, Period=1, Scale=0.4, Bias=1.5)
       B=&FNTimeOfDayScalar(Night Scale=0.5, Morning Scale=0.2, Evening Scale=0.3)
      ))
    )) {isSetParameters=true, module=SetVariables_#8}
   /SetVariables(Particles.LightRadius=80) {isSetParameters=true, module=SetVariables_#9}
   &ParticleState(Kill Particles When Lifetime Has Elapsed=false, Let Infinitely Lived Particles Die When Emitter Deactivates=true)
   &SpriteRotationRate(Rotation Rate=30)
   &ScaleSpriteSize(
    Scale Sprite Size Mode=::NewEnumerator0
    Uniform Scale Factor=&Multiply_Float(
     A=&Lerp_Float(A=1, B=1.5, Alpha=$Emitter.BurstAlpha)
     B=&Lerp_Float(A=1.1, B=4, Alpha=$System.ArtifactProximityToCamera)
    )
   )
   &ScaleMeshSize(Scale Factor=&VectorFromFloat(Value=&Lerp_Float(A=0.7, B=2.3, Alpha=$System.ArtifactProximityToCamera)))
   &SolveForcesAndVelocity(Clamp Velocity=false, Limit Acceleration=false)
   &ScaleColor(
    Scale Alpha=&Multiply_Float(
     A=&Multiply_Float(A=$System.ScaleAlpha, B=&OneMinusFloat(Float=$System.Timeline.NormalizedTime))
     B=&FNTimeOfDayScalar(Night Scale=0.5, Morning Scale=0.2, Evening Scale=0.3)
    )
    Scale RGB=&VectorFromFloat(Value=&Multiply_Float(
      A=&Multiply_Float(A=$System.ScaleAlpha, B=$System.GlowVariation)
      B=&Lerp_Float(A=1, B=10, Alpha=$Emitter.BurstAlpha)
     ))
    ScaleA=true
    ScaleRGB=true
   )
   &DynamicMaterialParameters(
    Index 0 Param 1=&MakeFloatFromVector(Vector=$Particles.GlowDirection, Channel=::NewEnumerator0)
    Index 0 Param 2=&MakeFloatFromVector(Vector=$Particles.GlowDirection, Channel=::NewEnumerator1)
    Index 0 Param 3=&MakeFloatFromVector(Vector=$Particles.GlowDirection, Channel=::NewEnumerator2)
    Index 0 Param 4=500
    Module.Index0Param0Write=true
    Module.Index0Param1Write=true
    Module.Index0Param2Write=true
    Module.Index0Param3Write=true
   )
   /SetVariables() {isSetParameters=true, module=SetVariables_#10}
  ]
  rendererClasses [
   &NiagaraSpriteRendererProperties
   &NiagaraLightRendererProperties
   &NiagaraMeshRendererProperties
  ]
  renderers [
   {rendererClass=&NiagaraSpriteRendererProperties}
   {rendererClass=&NiagaraLightRendererProperties, rendererIndex=1}
   {rendererClass=&NiagaraMeshRendererProperties, rendererIndex=2}
  ]
 }
 Glow_and_Light001 {
  bEnabled=false
  emitterUpdateScript [
   &EmitterState(
    Inactive Response=::NewEnumerator1
    Life Cycle Mode=::NewEnumerator1
    Loop Behavior=::NewEnumerator1
    Loop Duration Mode=::NewEnumerator1
    UseLoopDelay=false
   )
   &SpawnBurst_Instantaneous(Spawn Count=1, Use Spawn Probability=false)
   /SetVariables(Emitter.BurstAlpha:NiagaraFloat=&FloatFromCurve(
     FloatCurve=DataInterface() {propertyValues {Curve {keys [
         {arriveTangent=5.6877, interpMode=RCIM_Cubic, leaveTangent=5.6877, tangentMode=RCTM_User, time=0, value=0}
         {arriveTangent=-3.7229, interpMode=RCIM_Cubic, leaveTangent=-3.7229, tangentMode=RCTM_User, time=0.1, value=1}
         {interpMode=RCIM_Cubic, tangentMode=RCTM_User, time=0.5, value=0}
        ]}}}
     CurveIndex=$Emitter.Age
    )) {isSetParameters=true, module=SetVariables_#4}
  ]
  particleSpawnScript [
   &InitializeParticle(
    Color=<rgba 5 1 0 2>
    Lifetime=0.5
    Mesh Scale=<1.2 1 3>
    Mesh Scale Mode=::NewEnumerator4
    Position Offset=<-100 0 0>
    Sprite Size Mode=::NewEnumerator3
    Uniform Sprite Size=300
    UsePositionOffset=true
    Write Material Random=true
   ) {module=InitializeParticle001}
   /SetVariables(Particles.LightExponent=2.5) {isSetParameters=true, module=SetVariables_#5}
   /SetVariables(Particles.MeshOrientation=&MakeQuatFromAxisAngle(Axis=<0 1 0>, AngleInDegrees=90)) {isSetParameters=true, module=SetVariables_#6}
   /SetVariables(Particles.GlowDirection:Vector3f=&TransformVector(Vector=<1 0 0>, Destination Space=::World)) {isSetParameters=true, module=SetVariables_#7}
   &CameraOffset(Camera Offset Amount=30)
   &SystemLocation(MaskSpawnGroup=false, Offset=<30 0 0>)
  ]
  particleUpdateScript [
   /SetVariables(Particles.LightColor=&MultiplyLinearColor(
     Linear Color=<rgba 10 0.6667 0 1>
     Scale Factor=&MakeVector4_FromFloat(Float=&Multiply_Float(
       A=&Sine(Normalized Angle=$Emitter.Age, Period=1, Scale=0.4, Bias=1.5)
       B=&FNTimeOfDayScalar(Night Scale=0.5, Morning Scale=0.2, Evening Scale=0.3)
      ))
    )) {isSetParameters=true, module=SetVariables_#8}
   /SetVariables(Particles.LightRadius=80) {isSetParameters=true, module=SetVariables_#9}
   &ParticleState(Kill Particles When Lifetime Has Elapsed=false, Let Infinitely Lived Particles Die When Emitter Deactivates=true)
   &SpriteRotationRate(Rotation Rate=30)
   &ScaleSpriteSize(Scale Sprite Size Mode=::NewEnumerator0, Uniform Scale Factor=&Lerp_Float(A=1, B=1.5, Alpha=$Emitter.BurstAlpha))
   &SolveForcesAndVelocity(Clamp Velocity=false, Limit Acceleration=false)
   &ScaleColor(
    Scale Alpha=&Multiply_Float(
     A=&Multiply_Float(A=$System.ScaleAlpha, B=&OneMinusFloat(Float=$System.Timeline.NormalizedTime))
     B=&FNTimeOfDayScalar(Night Scale=0.5, Morning Scale=0.2, Evening Scale=0.3)
    )
    Scale RGB=&VectorFromFloat(Value=&Multiply_Float(
      A=&Multiply_Float(A=$System.ScaleAlpha, B=$System.GlowVariation)
      B=&Lerp_Float(A=1, B=10, Alpha=$Emitter.BurstAlpha)
     ))
    ScaleA=true
    ScaleRGB=true
   )
   &DynamicMaterialParameters(
    Index 0 Param 1=&MakeFloatFromVector(Vector=$Particles.GlowDirection, Channel=::NewEnumerator0)
    Index 0 Param 2=&MakeFloatFromVector(Vector=$Particles.GlowDirection, Channel=::NewEnumerator1)
    Index 0 Param 3=&MakeFloatFromVector(Vector=$Particles.GlowDirection, Channel=::NewEnumerator2)
    Index 0 Param 4=500
    Module.Index0Param0Write=true
    Module.Index0Param1Write=true
    Module.Index0Param2Write=true
    Module.Index0Param3Write=true
   )
   /SetVariables() {isSetParameters=true, module=SetVariables_#10}
  ]
  rendererClasses [
   &NiagaraSpriteRendererProperties
   &NiagaraLightRendererProperties
   &NiagaraMeshRendererProperties
  ]
  renderers [
   {rendererClass=&NiagaraSpriteRendererProperties}
   {rendererClass=&NiagaraLightRendererProperties, rendererIndex=1}
   {rendererClass=&NiagaraMeshRendererProperties, rendererIndex=2}
  ]
 }
 NE_FireTrail {
  emitterSpawnScript [/SetVariables(Emitter.ParticlePerRibbon=20, Emitter.RibbonNumber=1) {isSetParameters=true, module=SetVariables_#11}]
  emitterUpdateScript [
   &EmitterState(
    Inactive Response=::NewEnumerator1
    Life Cycle Mode=::NewEnumerator1
    Loop Behavior=::NewEnumerator1
    Loop Duration Mode=::NewEnumerator1
    UseLoopDelay=false
   )
   &SpawnBurst_Instantaneous(
    Spawn Count=&Multiply_Int(A=$Emitter.RibbonNumber, B=$Emitter.ParticlePerRibbon)
    Use Spawn Probability=false
    UseLoopCountLimit=false
   )
  ]
  particleSpawnScript [
   &PartitionParticles(
    ID to Partition:NiagaraInt32=&ReturnExecIndex()
    Number Of Partitions=$Emitter.RibbonNumber
    Particles Per Partition:NiagaraInt32=$Emitter.ParticlePerRibbon
    Partition Mode:ENiagara_PartitionMode=::NewEnumerator1
    Write Partition to Ribbon ID=::NewEnumerator0
   )
   /SetVariables() {isSetParameters=true, module=SetVariables_#12}
   &InitializeParticle(
    Color Mode=::NewEnumerator0
    Lifetime=5
    Ribbon Width=10
    Ribbon Width Mode=::NewEnumerator1
    UsePositionOffset=false
    Write Material Random=true
   )
   /SetVariables(Particles.RibbonU0Override=$Particles.PartitionParticles.Normalized Index In Partition) {isSetParameters=true, module=SetVariables_#13}
  ]
  particleUpdateScript [
   &ParticleState(Kill Particles When Lifetime Has Elapsed=false)
   /SetVariables(Particles.RibbonLength=2000) {isSetParameters=true, module=SetVariables_#14}
   /SetVariables(Particles.RibbonWidth=&Multiply_Float(
     A=$Particles.Initial.RibbonWidth
     B=&Lerp_Float(
      A=8
      B=20
      Alpha=&FloatFromCurve(
       FloatCurve=DataInterface() {propertyValues {Curve {keys [
           {arriveTangent=2.4963, interpMode=RCIM_Cubic, leaveTangent=2.4963, tangentMode=RCTM_User, time=0, value=0.3}
           {arriveTangent=5.8828, interpMode=RCIM_Cubic, leaveTangent=5.8828, tangentMode=RCTM_User, time=0.1, value=0.7}
           {arriveTangent=-0.0177, interpMode=RCIM_Cubic, leaveTangent=-0.0177, tangentMode=RCTM_User, time=0.7, value=2}
          ]}}}
       CurveIndex=$Particles.PartitionParticles.Normalized Index In Partition
       Scale Curve=&Lerp_Float(A=1, B=3, Alpha=$System.ArtifactProximityToCamera)
      )
     )
    )) {isSetParameters=true, module=SetVariables_#15}
   /SetVariables(Particles.GlowWidth:NiagaraFloat=&Multiply_Float(A=$Particles.RibbonWidth, B=&Lerp_Float(A=1.1, B=1.3, Alpha=$System.ArtifactProximityToCamera))) {isSetParameters=true, module=SetVariables_#16}
   /SetVariables(Particles.XOffsetValue:NiagaraFloat=&Add_Float(
     A=$Particles.XOffsetValue
     B=&Multiply_Float(A=$Engine.DeltaTime, B=&Lerp_Float(A=1.5, B=0.5, Alpha=$System.ArtifactProximityToCamera))
    )) {isSetParameters=true, module=SetVariables_#17}
   /SetVariables(Particles.GoalPos:NiagaraPosition=&AddVectorToPosition(
     Position=$Engine.Owner.Position
     Vector=&Multiply_VectorByFloat(
      Vector=&TransformVector(
       Vector=&Lerp_Vector(
        A=<1 0 0>
        B=&MakeVector(X=&Lerp_Float(A=1, B=5, Alpha=$System.ArtifactProximityToCamera), Y=0, Z=0.1)
        Alpha=&FloatFromCurve(
         FloatCurve=DataInterface() {propertyValues {Curve {keys [<at 0 0>,<at 1 1>]}}}
         CurveIndex=$Particles.PartitionParticles.Normalized Index In Partition
        )
       )
       Destination Space=::Simulation
      )
      Float=&Lerp_Float(A=0, B=$Particles.RibbonLength, Alpha=$Particles.PartitionParticles.Normalized Index In Partition)
     )
    )) {isSetParameters=true, module=SetVariables_#18}
   /SetVariables(Particles.Position=&Lerp_Vector(
     A=Linked() {linkedVariable {name=Particles.GoalPos, type=NiagaraPosition}}
     B=&Lerp_Vector(
      A=Linked() {linkedVariable {name=Particles.Position, type=NiagaraPosition}}
      B=Linked() {linkedVariable {name=Particles.GoalPos, type=NiagaraPosition}}
      Alpha=&Multiply_Float(
       A=$Engine.DeltaTime
       B=&Lerp_Float(
        A=15
        B=2
        Alpha=&FloatFromCurve(
         FloatCurve=DataInterface() {propertyValues {Curve {keys [
             {arriveTangent=3.7014, interpMode=RCIM_Cubic, leaveTangent=3.7014, tangentMode=RCTM_User, time=0, value=0}
             {arriveTangent=-0.0363, interpMode=RCIM_Cubic, leaveTangent=-0.0363, tangentMode=RCTM_User, time=1, value=1}
            ]}}}
         CurveIndex=$Particles.PartitionParticles.Normalized Index In Partition
        )
       )
      )
     )
     Alpha=&FloatFromCurve(
      FloatCurve=DataInterface() {propertyValues {Curve {keys [
          {arriveTangent=2.6466, interpMode=RCIM_Cubic, leaveTangent=2.6466, tangentMode=RCTM_User, time=0, value=0}
          {arriveTangent=0.2102, interpMode=RCIM_Cubic, leaveTangent=0.2102, tangentMode=RCTM_User, time=1, value=1}
         ]}}}
      CurveIndex=$Particles.PartitionParticles.Normalized Index In Partition
     )
    )) {isSetParameters=true, module=SetVariables_#19}
   &DynamicMaterialParameters(
    Index 0 Param 1=&Add_Float(
     A=/Game/Effects/Niagara/DynamicInputs/Random/FN_SeededRandomRangeFloat(Minimum:NiagaraFloat=1, Maximum:NiagaraFloat=5, Random Seed:NiagaraInt32=$Particles.PartitionParticles.Partition)
     B=$Particles.XOffsetValue
    )
    Index 0 Param 2=$System.SmoothDiveAlpha
    Index 1 Param 3=&FloatFromCurve(
     FloatCurve=DataInterface() {propertyValues {Curve {keys [<at 0.2 0>,<at 1 2>]}}}
     CurveIndex=&Add_Float(A=$Particles.PartitionParticles.Normalized Index In Partition, B=$System.Timeline.NormalizedTime)
    )
    Module.Index0Param0Write=true
    Module.Index0Param1Write=true
    Module.Index0Param2Write=true
    Module.Index0Param3Write=true
    Module.Index1Param0Write=true
    Module.Index1Param1Write=true
    Module.Index1Param2Write=true
    Module.Index1Param3Write=false
    Write Parameter Index 1=true
   )
   &ScaleColor(
    Scale Alpha=&Multiply_Float(
     A=&FloatFromCurve(
      FloatCurve=DataInterface() {propertyValues {Curve {keys [
          {arriveTangent=-1.6635, interpMode=RCIM_Cubic, leaveTangent=-1.6635, tangentMode=RCTM_User, time=0.15, value=1}
          {arriveTangent=-0.3046, interpMode=RCIM_Cubic, leaveTangent=-0.3046, tangentMode=RCTM_User, time=0.8, value=0}
         ]}}}
      CurveIndex=$Particles.PartitionParticles.Normalized Index In Partition
     )
     B=1
    )
    Scale RGB=&VectorFromFloat(Value=&FloatFromCurve(
      FloatCurve=DataInterface() {propertyValues {Curve {keys [
          {arriveTangent=-26.9347, interpMode=RCIM_Cubic, leaveTangent=-26.9347, tangentMode=RCTM_User, time=0, value=7}
          {interpMode=RCIM_Cubic, tangentMode=RCTM_User, time=0.5, value=1}
         ]}}}
      CurveIndex=$Particles.PartitionParticles.Normalized Index In Partition
     ))
    ScaleA=true
    ScaleRGB=true
   )
   /SetVariables(Particles.ColorGlow:LinearColor=&MultiplyLinearColor(
     Linear Color=$Particles.Color
     Scale Factor=&MakeFloat4FromVectorAndFloat(
      "Vector (RGB)"=&VectorFromFloat(Value=&Lerp_Float(A=1, B=3, Alpha=$System.ArtifactProximityToCamera))
      "Float (Alpha)"=1
     )
    )) {isSetParameters=true}
  ]
  rendererClasses [&NiagaraRibbonRendererProperties]
  renderers [
   {rendererClass=&NiagaraRibbonRendererProperties}
   {rendererClass=&NiagaraRibbonRendererProperties, rendererIndex=1}
   {rendererClass=&NiagaraRibbonRendererProperties, rendererIndex=2}
  ]
 }
 NE_FireTrail001 {
  bEnabled=false
  emitterSpawnScript [/SetVariables(Emitter.ParticlePerRibbon=20, Emitter.RibbonNumber=1) {isSetParameters=true, module=SetVariables_#11}]
  emitterUpdateScript [
   &EmitterState(
    Inactive Response=::NewEnumerator1
    Life Cycle Mode=::NewEnumerator1
    Loop Behavior=::NewEnumerator1
    Loop Duration Mode=::NewEnumerator1
    UseLoopDelay=false
   )
   &SpawnBurst_Instantaneous(
    Spawn Count=&Multiply_Int(A=$Emitter.RibbonNumber, B=$Emitter.ParticlePerRibbon)
    Use Spawn Probability=false
    UseLoopCountLimit=false
   )
  ]
  particleSpawnScript [
   &PartitionParticles(
    ID to Partition:NiagaraInt32=&ReturnExecIndex()
    Number Of Partitions=$Emitter.RibbonNumber
    Particles Per Partition:NiagaraInt32=$Emitter.ParticlePerRibbon
    Partition Mode:ENiagara_PartitionMode=::NewEnumerator1
    Write Partition to Ribbon ID=::NewEnumerator0
   )
   /SetVariables(Particles.XOffsetValue:NiagaraFloat=&SelectFloatFromArray(Direct Array Index=$Particles.PartitionParticles.Partition, Float Selection Array=DataInterface() {propertyValues {FloatData [0,0,0,0]}})) {isSetParameters=true, module=SetVariables_#12}
   &InitializeParticle(
    Color Mode=::NewEnumerator0
    Lifetime=5
    Ribbon Width=10
    Ribbon Width Mode=::NewEnumerator1
    Sprite Size Mode=::NewEnumerator3
    Uniform Sprite Size=5
    UsePositionOffset=false
    Write Material Random=true
   )
   /SetVariables(Particles.RibbonU0Override=$Particles.PartitionParticles.Normalized Index In Partition) {isSetParameters=true, module=SetVariables_#13}
  ]
  particleUpdateScript [
   &ParticleState(Kill Particles When Lifetime Has Elapsed=false)
   /SetVariables(Particles.RibbonLength=2000) {isSetParameters=true, module=SetVariables_#14}
   /SetVariables(Particles.RibbonWidth=&Multiply_Float(
     A=$Particles.Initial.RibbonWidth
     B=&Lerp_Float(
      A=8
      B=20
      Alpha=&FloatFromCurve(
       FloatCurve=DataInterface() {propertyValues {Curve {keys [
           {arriveTangent=2.4963, interpMode=RCIM_Cubic, leaveTangent=2.4963, tangentMode=RCTM_User, time=0, value=0.1}
           {arriveTangent=5.8828, interpMode=RCIM_Cubic, leaveTangent=5.8828, tangentMode=RCTM_User, time=0.1, value=0.7}
           {arriveTangent=-0.0177, interpMode=RCIM_Cubic, leaveTangent=-0.0177, tangentMode=RCTM_User, time=0.7, value=2}
          ]}}}
       CurveIndex=$Particles.PartitionParticles.Normalized Index In Partition
      )
     )
    )) {isSetParameters=true, module=SetVariables_#15}
   /SetVariables(Particles.GlowWidth:NiagaraFloat=&Multiply_Float(A=$Particles.RibbonWidth, B=1.1)) {isSetParameters=true, module=SetVariables_#16}
   /SetVariables(Particles.XOffsetValue:NiagaraFloat=&Add_Float(
     A=$Particles.XOffsetValue
     B=&Multiply_Float(A=$Engine.DeltaTime, B=&Lerp_Float(A=0.4, B=1.2, Alpha=$System.SmoothDiveAlpha))
    )) {isSetParameters=true, module=SetVariables_#17}
   /SetVariables(Particles.GoalPos:NiagaraPosition=&AddVectorToPosition(
     Position=$Engine.Owner.Position
     Vector=&Multiply_VectorByFloat(
      Vector=&TransformVector(
       Vector=&Lerp_Vector(
        A=<1 0 0>
        B=<1 0 0.1>
        Alpha=&FloatFromCurve(
         FloatCurve=DataInterface() {propertyValues {Curve {keys [<at 0 0>,<at 1 1>]}}}
         CurveIndex=$Particles.PartitionParticles.Normalized Index In Partition
        )
       )
       Destination Space=::Simulation
      )
      Float=&Lerp_Float(A=0, B=$Particles.RibbonLength, Alpha=$Particles.PartitionParticles.Normalized Index In Partition)
     )
    )) {isSetParameters=true, module=SetVariables_#18}
   /SetVariables(Particles.Position=&Lerp_Vector(
     A=Linked() {linkedVariable {name=Particles.GoalPos, type=NiagaraPosition}}
     B=&Lerp_Vector(
      A=Linked() {linkedVariable {name=Particles.Position, type=NiagaraPosition}}
      B=Linked() {linkedVariable {name=Particles.GoalPos, type=NiagaraPosition}}
      Alpha=&Multiply_Float(
       A=$Engine.DeltaTime
       B=&Lerp_Float(
        A=15
        B=2
        Alpha=&FloatFromCurve(
         FloatCurve=DataInterface() {propertyValues {Curve {keys [
             {arriveTangent=3.7014, interpMode=RCIM_Cubic, leaveTangent=3.7014, tangentMode=RCTM_User, time=0, value=0}
             {arriveTangent=-0.0363, interpMode=RCIM_Cubic, leaveTangent=-0.0363, tangentMode=RCTM_User, time=1, value=1}
            ]}}}
         CurveIndex=$Particles.PartitionParticles.Normalized Index In Partition
        )
       )
      )
     )
     Alpha=&FloatFromCurve(
      FloatCurve=DataInterface() {propertyValues {Curve {keys [
          {arriveTangent=2.6466, interpMode=RCIM_Cubic, leaveTangent=2.6466, tangentMode=RCTM_User, time=0, value=0}
          {arriveTangent=0.2102, interpMode=RCIM_Cubic, leaveTangent=0.2102, tangentMode=RCTM_User, time=1, value=1}
         ]}}}
      CurveIndex=$Particles.PartitionParticles.Normalized Index In Partition
     )
    )) {isSetParameters=true, module=SetVariables_#19}
   &DynamicMaterialParameters(
    Index 0 Param 1=&Add_Float(
     A=&FN_SeededRandomRangeFloat(Minimum:NiagaraFloat=1, Maximum:NiagaraFloat=5, Random Seed:NiagaraInt32=$Particles.PartitionParticles.Partition)
     B=$Particles.XOffsetValue
    )
    Index 0 Param 2=$System.SmoothDiveAlpha
    Index 1 Param 3=&FloatFromCurve(
     FloatCurve=DataInterface() {propertyValues {Curve {keys [<at 0.2 0>,<at 1 2>]}}}
     CurveIndex=&Add_Float(A=$Particles.PartitionParticles.Normalized Index In Partition, B=$System.Timeline.NormalizedTime)
    )
    Module.Index0Param0Write=true
    Module.Index0Param1Write=true
    Module.Index0Param2Write=true
    Module.Index0Param3Write=true
    Module.Index1Param0Write=true
    Module.Index1Param1Write=true
    Module.Index1Param2Write=true
    Module.Index1Param3Write=true
    Write Parameter Index 1=true
   )
   &ScaleColor(
    Scale Alpha=&Multiply_Float(
     A=&FloatFromCurve(
      FloatCurve=DataInterface() {propertyValues {Curve {keys [
          {arriveTangent=-1.6635, interpMode=RCIM_Cubic, leaveTangent=-1.6635, tangentMode=RCTM_User, time=0.15, value=1}
          {arriveTangent=-0.3046, interpMode=RCIM_Cubic, leaveTangent=-0.3046, tangentMode=RCTM_User, time=0.7, value=0}
         ]}}}
      CurveIndex=$Particles.PartitionParticles.Normalized Index In Partition
     )
     B=1
    )
    ScaleA=true
    ScaleRGB=true
   )
  ]
  rendererClasses [&NiagaraRibbonRendererProperties]
  renderers [
   {rendererClass=&NiagaraRibbonRendererProperties}
   {rendererClass=&NiagaraRibbonRendererProperties, rendererIndex=1}
   {rendererClass=&NiagaraRibbonRendererProperties, rendererIndex=2}
  ]
 }
 NE_FlashFlare {
  emitterUpdateScript [
   &EmitterState(
    Inactive Response=::NewEnumerator1
    Life Cycle Mode=::NewEnumerator1
    Loop Behavior=::NewEnumerator1
    Loop Duration Mode=::NewEnumerator1
    UseLoopDelay=false
   )
   &SpawnBurst_Instantaneous(Spawn Count=2, Use Spawn Probability=false, UseLoopCountLimit=false)
   &SpawnBurst_Instantaneous(Spawn Group=1, Use Spawn Probability=false, UseLoopCountLimit=false) {module=SpawnBurst_Instantaneous001}
  ]
  particleSpawnScript [
   &InitializeParticle(
    Color=&MaskLinearColorBySpawnGroup(Masked Linear Color=<rgba 1 0.3634 0 1>, Passthrough Linear Color=<rgba 1 0.1834 0 1>)
    Lifetime=0.5
    Mass Mode=::NewEnumerator1
    Mesh Scale Mode=::NewEnumerator3
    Mesh Uniform Scale=0.7
    Position=&SimulationPosition()
    Position Mode=::NewEnumerator1
    Sprite Rotation Angle=&MaskFloatBySpawnGroup(
     Masked Float=&Lerp_Float(A=0.125, B=0.375, Alpha=&ReturnNormalizedExecIndex())
     Passthrough Float=&RandomRangeFloat(Minimum=0, Maximum=1)
    )
    Sprite Rotation Mode=::NewEnumerator1
    Sprite Size=&MaskVector2dBySpawnGroup(
     Masked Vector2D=&Lerp_Vector2(A=<4500 350>, B=<2500 350>, Alpha=&ReturnNormalizedExecIndex())
     Passthrough Vector2D=&Vector2DFromFloat(Value=3500)
     Spawn Group Mask=0
    )
    Sprite Size Mode=::NewEnumerator4
    UsePositionOffset=false
    Write Material Random=true
   ) {module=InitializeParticle001}
   &CameraOffset(Camera Offset Amount=&Lerp_Float(A=150, B=-25, Alpha=$Particles.HandFloat)) {enabled=false}
  ]
  particleUpdateScript [
   &CameraOffset(Camera Offset Amount=&Lerp_Float(A=-200, B=50, Alpha=$System.ArtifactProximityToCamera)) {module=CameraOffset001}
   &ParticleState(Kill Particles When Lifetime Has Elapsed=false)
   &SolveForcesAndVelocity(Clamp Velocity=false, Limit Acceleration=false)
   &ScaleColor(
    Scale Alpha=&FloatFromCurve(
     FloatCurve=DataInterface() {propertyValues {Curve {keys [
         {arriveTangent=6.4019, interpMode=RCIM_Cubic, leaveTangent=6.4019, tangentMode=RCTM_User, time=0, value=0}
         {arriveTangent=-1.6579, interpMode=RCIM_Cubic, leaveTangent=-1.6579, tangentMode=RCTM_User, time=0.2, value=1}
         {arriveTangent=-0.1886, interpMode=RCIM_Cubic, leaveTangent=-0.1886, tangentMode=RCTM_User, time=1, value=0.3}
        ]}}}
     Scale Curve=&Multiply_Float(
      A=&Lerp_Float(A=0.5, B=1.7, Alpha=$System.ArtifactProximityToCamera)
      B=&FloatFromCurve(FloatCurve=DataInterface() {propertyValues {Curve {keys [<at 0 1>,<at 1 0>]}}}, CurveIndex=$System.Timeline.NormalizedTime)
     )
    )
    Scale RGB=&MakeVectorFromLinearColorRGB(Color=&ColorFromCurve(ColorCurve=DataInterface() {propertyValues {
        AlphaCurve {keys [<at 0 1 RCIM_Cubic>]}
        BlueCurve {keys [<at 0.2687 1 RCIM_Cubic>,<at 0.8844 0.05 RCIM_Cubic>]}
        GreenCurve {keys [<at 0.2687 1 RCIM_Cubic>,<at 0.8844 0.24 RCIM_Cubic>]}
        RedCurve {keys [<at 0.2687 1 RCIM_Cubic>,<at 0.8844 1 RCIM_Cubic>]}
       }}))
    ScaleA=true
    ScaleRGB=true
   )
   &ScaleSpriteSize(Scale Factor=&Multiply_Vector2D_ByFloat(
     Float=&Multiply_Float(
      A=&FloatFromCurve(
       FloatCurve=DataInterface() {propertyValues {Curve {keys [
           {arriveTangent=0.7585, interpMode=RCIM_Cubic, leaveTangent=0.7585, tangentMode=RCTM_User, time=0, value=0}
           {interpMode=RCIM_Cubic, tangentMode=RCTM_User, time=0.6, value=1}
           <at 1 0.7 RCIM_Cubic>
          ]}}}
       Scale Curve=2
      )
      B=&Lerp_Float(A=0.2, B=1.5, Alpha=$System.ArtifactProximityToCamera)
     )
     Vector2D=<1 1>
    ))
   &ScaleSpriteSize(
    Scale Sprite Size Mode=::NewEnumerator0
    Uniform Scale Factor=&Waveform(
     "Waveform [1]"=::NewEnumerator0
     "[1] Phase Scale"=&Waveform(Global Input Position=0.2, "Waveform [1]"=::NewEnumerator0, "[1] Frequency"=0.26, "[1] Amplitude Min/Max"=<0 0.3>)
     "[1] Frequency"=0.1
     "[1] Amplitude Min/Max"=<0.7 1.1>
    )
   ) {module=ScaleSpriteSize001}
   &SpriteRotationRate(Rotation Rate=&UniformRangedFloat(Minimum=-50, Maximum=50))
   /SetVariables(Particles.RayColor:LinearColor=&MultiplyLinearColor(
     Linear Color=&MultiplyLinearColor(
      Linear Color=$Particles.Color
      Scale Factor=&MakeVector4_FromLinearColor(LinearColor=<rgba 1.25 1.25 1.25 1>)
     )
     Scale Factor=&MakeFloat4FromVectorAndFloat(
      "Vector (RGB)"=<1 1 1>
      "Float (Alpha)"=&FloatFromCurve(FloatCurve=DataInterface() {propertyValues {Curve {keys [<at 0.5 1>,<at 1 0>]}}})
     )
    )) {enabled=false, isSetParameters=true, module=SetVariables_#20}
   /SetVariables(Particles.DarkGlowSize:Vector2f=&Multiply_Vector2D_ByFloat(Float=1, Vector2D=$Particles.SpriteSize)) {enabled=false, isSetParameters=true, module=SetVariables_#21}
   /SetVariables(Particles.DargGlowColor:LinearColor=&ColorFromCurve(ColorCurve=DataInterface() {propertyValues {
       AlphaCurve {keys [<at 0.4134 1 RCIM_Cubic>,<at 1 0 RCIM_Cubic>]}
       BlueCurve {keys [<at 0 0.0343 RCIM_Cubic>]}
       GreenCurve {keys [<at 0 0.1 RCIM_Cubic>]}
       RedCurve {keys [<at 0 0.0372 RCIM_Cubic>]}
      }})) {enabled=false, isSetParameters=true, module=SetVariables_#22}
  ]
  rendererClasses [&NiagaraSpriteRendererProperties]
  renderers [
   {rendererClass=&NiagaraSpriteRendererProperties}
   {rendererClass=&NiagaraSpriteRendererProperties, rendererIndex=1}
  ]
 }
 NE_FlashFlare001 {
  bEnabled=false
  emitterUpdateScript [
   &EmitterState(Inactive Response=::NewEnumerator1, Life Cycle Mode=::NewEnumerator1, Loop Behavior=::NewEnumerator1, Loop Duration=2, UseLoopDelay=false)
   &SpawnBurst_Instantaneous(Spawn Count=2, Use Spawn Probability=false, UseLoopCountLimit=false)
   &SpawnBurst_Instantaneous(Spawn Group=1, Use Spawn Probability=false, UseLoopCountLimit=false) {module=SpawnBurst_Instantaneous001}
  ]
  particleSpawnScript [
   &InitializeParticle(
    Color=&MaskLinearColorBySpawnGroup(Masked Linear Color=<rgba 1 0.3634 0 1>, Passthrough Linear Color=<rgba 1 0.1834 0 1>)
    Lifetime=0.35
    Mass Mode=::NewEnumerator1
    Mesh Scale Mode=::NewEnumerator3
    Mesh Uniform Scale=0.7
    Position=&SimulationPosition()
    Position Mode=::NewEnumerator1
    Sprite Rotation Angle=&MaskFloatBySpawnGroup(
     Masked Float=&Lerp_Float(A=0.125, B=0.375, Alpha=&ReturnNormalizedExecIndex())
     Passthrough Float=&RandomRangeFloat(Minimum=0, Maximum=1)
    )
    Sprite Rotation Mode=::NewEnumerator1
    Sprite Size=&MaskVector2dBySpawnGroup(
     Masked Vector2D=&Lerp_Vector2(A=<4500 350>, B=<2500 350>, Alpha=&ReturnNormalizedExecIndex())
     Passthrough Vector2D=&Vector2DFromFloat(Value=3500)
     Spawn Group Mask=0
    )
    Sprite Size Mode=::NewEnumerator4
    UsePositionOffset=true
    Write Material Random=true
   ) {module=InitializeParticle001}
   &CameraOffset(Camera Offset Amount=&Lerp_Float(A=150, B=-25, Alpha=$Particles.HandFloat))
  ]
  particleUpdateScript [
   &ParticleState()
   &SolveForcesAndVelocity(Clamp Velocity=false, Limit Acceleration=false)
   &ScaleColor(
    Scale Alpha=&FloatFromCurve(FloatCurve=DataInterface() {propertyValues {Curve {keys [
         {arriveTangent=6.4019, interpMode=RCIM_Cubic, leaveTangent=6.4019, tangentMode=RCTM_User, time=0, value=0}
         {arriveTangent=0.002, interpMode=RCIM_Cubic, leaveTangent=0.002, time=0.2, value=0.7}
         {arriveTangent=-0.9006, interpMode=RCIM_Cubic, leaveTangent=-0.9006, tangentMode=RCTM_User, time=1.0011, value=0.002}
        ]}}})
    ScaleA=true
    ScaleRGB=false
   )
   &ScaleSpriteSize(Scale Factor=&Multiply_Vector2D_ByFloat(
     Float=&FloatFromCurve(FloatCurve=DataInterface() {propertyValues {Curve {keys [
          {arriveTangent=0.7585, interpMode=RCIM_Cubic, leaveTangent=0.7585, tangentMode=RCTM_User, time=0, value=0}
          {interpMode=RCIM_Cubic, tangentMode=RCTM_User, time=1, value=1}
         ]}}})
     Vector2D=<1 1>
    ))
   &SpriteRotationRate(Rotation Rate=&UniformRangedFloat(Minimum=-50, Maximum=50))
   /SetVariables(Particles.RayColor:LinearColor=&MultiplyLinearColor(
     Linear Color=&MultiplyLinearColor(
      Linear Color=$Particles.Color
      Scale Factor=&MakeVector4_FromLinearColor(LinearColor=<rgba 1.25 1.25 1.25 1>)
     )
     Scale Factor=&MakeFloat4FromVectorAndFloat(
      "Vector (RGB)"=<1 1 1>
      "Float (Alpha)"=&FloatFromCurve(FloatCurve=DataInterface() {propertyValues {Curve {keys [<at 0.5 1>,<at 1 0>]}}})
     )
    )) {isSetParameters=true, module=SetVariables_#20}
   /SetVariables(Particles.DarkGlowSize:Vector2f=&Multiply_Vector2D_ByFloat(Float=1, Vector2D=$Particles.SpriteSize)) {isSetParameters=true, module=SetVariables_#21}
   /SetVariables(Particles.DargGlowColor:LinearColor=&ColorFromCurve(ColorCurve=DataInterface() {propertyValues {
       AlphaCurve {keys [<at 0.4134 1 RCIM_Cubic>,<at 1 0 RCIM_Cubic>]}
       BlueCurve {keys [<at 0 0.0343 RCIM_Cubic>]}
       GreenCurve {keys [<at 0 0.1 RCIM_Cubic>]}
       RedCurve {keys [<at 0 0.0372 RCIM_Cubic>]}
      }})) {isSetParameters=true, module=SetVariables_#22}
  ]
  rendererClasses [&NiagaraSpriteRendererProperties]
  renderers [
   {rendererClass=&NiagaraSpriteRendererProperties}
   {rendererClass=&NiagaraSpriteRendererProperties, rendererIndex=1}
  ]
 }
 NE_FlashFlare002 {
  bEnabled=false
  emitterUpdateScript [
   &EmitterState(
    Inactive Response=::NewEnumerator1
    Life Cycle Mode=::NewEnumerator1
    Loop Behavior=::NewEnumerator1
    Loop Duration Mode=::NewEnumerator1
    UseLoopDelay=false
   )
   &SpawnBurst_Instantaneous(Spawn Count=2, Use Spawn Probability=false, UseLoopCountLimit=false)
   &SpawnBurst_Instantaneous(Spawn Group=1, Use Spawn Probability=false, UseLoopCountLimit=false) {module=SpawnBurst_Instantaneous001}
  ]
  particleSpawnScript [
   &InitializeParticle(
    Color=&MaskLinearColorBySpawnGroup(Masked Linear Color=<rgba 1 0.3634 0 1>, Passthrough Linear Color=<rgba 1 0.1834 0 1>)
    Lifetime=0.5
    Mass Mode=::NewEnumerator1
    Mesh Scale Mode=::NewEnumerator3
    Mesh Uniform Scale=0.7
    Position=&SimulationPosition()
    Position Mode=::NewEnumerator1
    Sprite Rotation Angle=&MaskFloatBySpawnGroup(
     Masked Float=&Lerp_Float(A=0.125, B=0.375, Alpha=&ReturnNormalizedExecIndex())
     Passthrough Float=&RandomRangeFloat(Minimum=0, Maximum=1)
    )
    Sprite Rotation Mode=::NewEnumerator1
    Sprite Size=&MaskVector2dBySpawnGroup(
     Masked Vector2D=&Lerp_Vector2(A=<4500 350>, B=<2500 350>, Alpha=&ReturnNormalizedExecIndex())
     Passthrough Vector2D=&Vector2DFromFloat(Value=3500)
     Spawn Group Mask=0
    )
    Sprite Size Mode=::NewEnumerator4
    UsePositionOffset=true
    Write Material Random=true
   ) {module=InitializeParticle001}
   &CameraOffset(Camera Offset Amount=&Lerp_Float(A=150, B=-25, Alpha=$Particles.HandFloat)) {enabled=false}
  ]
  particleUpdateScript [
   &CameraOffset(Camera Offset Amount=&Lerp_Float(A=-200, B=50, Alpha=$System.ArtifactProximityToCamera)) {module=CameraOffset001}
   &ParticleState(Kill Particles When Lifetime Has Elapsed=false)
   &SolveForcesAndVelocity(Clamp Velocity=false, Limit Acceleration=false)
   &ScaleColor(
    Scale Alpha=&FloatFromCurve(
     FloatCurve=DataInterface() {propertyValues {Curve {keys [
         {arriveTangent=6.4019, interpMode=RCIM_Cubic, leaveTangent=6.4019, tangentMode=RCTM_User, time=0, value=0}
         {arriveTangent=-1.6579, interpMode=RCIM_Cubic, leaveTangent=-1.6579, tangentMode=RCTM_User, time=0.2, value=1}
         {arriveTangent=-0.1886, interpMode=RCIM_Cubic, leaveTangent=-0.1886, tangentMode=RCTM_User, time=1, value=0.3}
        ]}}}
     Scale Curve=&Lerp_Float(A=0, B=1, Alpha=$System.ArtifactProximityToCamera)
    )
    Scale RGB=&MakeVectorFromLinearColorRGB(Color=&ColorFromCurve(ColorCurve=DataInterface() {propertyValues {
        AlphaCurve {keys [<at 0 1 RCIM_Cubic>]}
        BlueCurve {keys [<at 0.2687 1 RCIM_Cubic>,<at 0.8844 0.05 RCIM_Cubic>]}
        GreenCurve {keys [<at 0.2687 1 RCIM_Cubic>,<at 0.8844 0.5725 RCIM_Cubic>]}
        RedCurve {keys [<at 0.2687 1 RCIM_Cubic>,<at 0.8844 1 RCIM_Cubic>]}
       }}))
    ScaleA=true
    ScaleRGB=true
   )
   &ScaleSpriteSize(Scale Factor=&Multiply_Vector2D_ByFloat(
     Float=&Multiply_Float(
      A=&FloatFromCurve(
       FloatCurve=DataInterface() {propertyValues {Curve {keys [
           {arriveTangent=0.7585, interpMode=RCIM_Cubic, leaveTangent=0.7585, tangentMode=RCTM_User, time=0, value=0}
           {interpMode=RCIM_Cubic, tangentMode=RCTM_User, time=0.6, value=1}
           <at 1 0.7 RCIM_Cubic>
          ]}}}
       Scale Curve=2
      )
      B=&Lerp_Float(A=0.1, B=1, Alpha=$System.ArtifactProximityToCamera)
     )
     Vector2D=<1 1>
    ))
   &ScaleSpriteSize(
    Scale Sprite Size Mode=::NewEnumerator0
    Uniform Scale Factor=&Waveform(
     "Waveform [1]"=::NewEnumerator0
     "[1] Phase Scale"=&Waveform(Global Input Position=0.2, "Waveform [1]"=::NewEnumerator0, "[1] Frequency"=0.26, "[1] Amplitude Min/Max"=<0 0.3>)
     "[1] Frequency"=0.1
     "[1] Amplitude Min/Max"=<0.7 1.1>
    )
   ) {module=ScaleSpriteSize001}
   &SpriteRotationRate(Rotation Rate=&UniformRangedFloat(Minimum=-50, Maximum=50))
   /SetVariables(Particles.RayColor:LinearColor=&MultiplyLinearColor(
     Linear Color=&MultiplyLinearColor(
      Linear Color=$Particles.Color
      Scale Factor=&MakeVector4_FromLinearColor(LinearColor=<rgba 1.25 1.25 1.25 1>)
     )
     Scale Factor=&MakeFloat4FromVectorAndFloat(
      "Vector (RGB)"=<1 1 1>
      "Float (Alpha)"=&FloatFromCurve(FloatCurve=DataInterface() {propertyValues {Curve {keys [<at 0.5 1>,<at 1 0>]}}})
     )
    )) {isSetParameters=true, module=SetVariables_#20}
   /SetVariables(Particles.DarkGlowSize:Vector2f=&Multiply_Vector2D_ByFloat(Float=1, Vector2D=$Particles.SpriteSize)) {isSetParameters=true, module=SetVariables_#21}
   /SetVariables(Particles.DargGlowColor:LinearColor=&ColorFromCurve(ColorCurve=DataInterface() {propertyValues {
       AlphaCurve {keys [<at 0.4134 1 RCIM_Cubic>,<at 1 0 RCIM_Cubic>]}
       BlueCurve {keys [<at 0 0.0343 RCIM_Cubic>]}
       GreenCurve {keys [<at 0 0.1 RCIM_Cubic>]}
       RedCurve {keys [<at 0 0.0372 RCIM_Cubic>]}
      }})) {isSetParameters=true, module=SetVariables_#22}
  ]
  rendererClasses [&NiagaraSpriteRendererProperties]
  renderers [
   {rendererClass=&NiagaraSpriteRendererProperties}
   {rendererClass=&NiagaraSpriteRendererProperties, rendererIndex=1}
  ]
 }
 NE_SpriteSmoke {
  emitterUpdateScript [
   &EmitterState(Life Cycle Mode=::NewEnumerator1, Loop Duration=1, UseLoopDelay=false)
   &SpawnRate(
    SpawnRate=&MakeCustomFloatFromBool(Bool=$User.FadeOut, True Float=0, False Float=20)
    Use Spawn Probability=false
   )
  ]
  particleSpawnScript [
   &InitializeParticle(
    Lifetime=&Multiply_Float(
     A=&RandomRangeFloat(Minimum=0.8, Maximum=1)
     B=&Lerp_Float(A=1, B=2.2, Alpha=$System.ArtifactProximityToCamera)
    )
    Sprite Rotation Mode=::NewEnumerator2
    Sprite Size Mode=::NewEnumerator3
    Uniform Sprite Size=1
    UsePositionOffset=false
    Write Material Random=true
   )
   /SetVariables() {isSetParameters=true, module=SetVariables_#23}
   &CameraOffset(Camera Offset Amount=-10)
   &AddVelocity(Velocity=<1 0 0>, Velocity Speed Scale=&RandomRangeFloat(Minimum=1550, Maximum=1850))
  ]
  particleUpdateScript [
   &self:FollowRibbon(
    Alpha:NiagaraFloat=&FloatFromCurve(FloatCurve=DataInterface() {propertyValues {Curve {keys [<at 0 0>,<at 1 0.8>]}}})
    ParticlePerRibbon=20
    RibbonReader=DataInterface() {propertyValues {EmitterBinding {bindingMode=Other, emitterName=NE_FireTrail}}}
   ) {enabled=false, module=FollowRibbon}
   /SetVariables(Particles.SpriteSize=&Vector2DFromFloat(Value=&Multiply_Float(
      A=&Lerp_Float(
       A=80
       B=450
       Alpha=&FloatFromCurve(FloatCurve=DataInterface() {propertyValues {Curve {keys [
            {arriveTangent=0.7052, interpMode=RCIM_Cubic, leaveTangent=0.7052, tangentMode=RCTM_User, time=0, value=0}
            {arriveTangent=1.0865, interpMode=RCIM_Cubic, leaveTangent=1.0865, tangentMode=RCTM_User, time=1, value=1}
           ]}}})
      )
      B=&Lerp_Float(A=1.1, B=2, Alpha=$System.ArtifactProximityToCamera)
     ))) {isSetParameters=true, module=SetVariables_#24}
   &DynamicMaterialParameters(
    Index 0 Param 1=&Add_Float(
     A=&Multiply_Float(A=$Particles.NormalizedAge, B=1)
     B=&RandomRangeFloat(Minimum=0, Maximum=1)
    )
    Index 0 Param 3=10
    Index 1 Param 1=1
    Index 1 Param 3=1
    Index 1 Param 4=0.5
    Index 2 Param 1=1
    Index 2 Param 2=&RandomRangeFloat(Minimum=0, Maximum=1)
    Index 2 Param 3=&FloatFromCurve(
     FloatCurve=DataInterface() {propertyValues {Curve {keys [
         {arriveTangent=0.8418, interpMode=RCIM_Cubic, leaveTangent=0.8418, tangentMode=RCTM_User, time=0, value=0}
         {arriveTangent=1.2208, interpMode=RCIM_Cubic, leaveTangent=1.2208, tangentMode=RCTM_User, time=0.9, value=1}
        ]}}}
     CurveIndex=&Add_Float(A=$Particles.NormalizedAge, B=$System.Timeline.NormalizedTime)
     Scale Curve=5
    )
    Index 2 Param 4=0.7
    Module.Param0WriteEnabled=true
    Module.Param0WriteEnabled1=true
    Module.Param0WriteEnabled2=true
    Module.Param1WriteEnabled=true
    Module.Param1WriteEnabled1=true
    Module.Param1WriteEnabled2=true
    Module.Param2WriteEnabled=true
    Module.Param2WriteEnabled1=true
    Module.Param2WriteEnabled2=true
    Module.Param3WriteEnabled=true
    Module.Param3WriteEnabled1=true
    Module.Param3WriteEnabled2=true
    Write Parameter Index 1=true
    Write Parameter Index 2=true
   )
   &ParticleState()
   &ScaleColor(
    Scale Alpha=&Multiply_Float(
     A=&FloatFromCurve(
      FloatCurve=DataInterface() {propertyValues {Curve {keys [
          {arriveTangent=15.5597, interpMode=RCIM_Cubic, leaveTangent=15.5597, tangentMode=RCTM_User, time=0.1, value=0}
          {arriveTangent=0.3912, interpMode=RCIM_Cubic, tangentMode=RCTM_Break, time=0.2, value=1}
          {arriveTangent=-0.1085, interpMode=RCIM_Cubic, leaveTangent=-0.1085, tangentMode=RCTM_User, time=0.75, value=1}
          <at 1 0>
         ]}}}
      CurveIndex=&Add_Float(A=$Particles.NormalizedAge, B=$System.Timeline.NormalizedTime)
     )
     B=1
    )
    Scale RGB=&MakeVectorFromLinearColorRGB(Color=&ColorFromCurve(ColorCurve=DataInterface() {propertyValues {
        AlphaCurve {keys []}
        BlueCurve {keys [<at 0.0774 0.0625>,<at 0.3614 0>,<at 0.6427 0 RCIM_Cubic>,<at 0.9972 0 RCIM_Cubic>]}
        GreenCurve {keys [
          <at 0.0774 0.3826>
          <at 0.3614 0.0321>
          {arriveTangent=-0.0437, interpMode=RCIM_Cubic, leaveTangent=-0.0437, time=0.6427, value=0.0226}
          <at 0.9972 0.0044 RCIM_Cubic>
         ]}
        RedCurve {keys [
          <at 0.0774 2>
          <at 0.3614 0.2>
          {arriveTangent=-0.2517, interpMode=RCIM_Cubic, leaveTangent=-0.2517, time=0.6427, value=0.1}
          <at 0.9972 0.04 RCIM_Cubic>
         ]}
       }}))
    ScaleA=true
    ScaleRGB=true
   )
   &AccelerationForce(Acceleration=&MakeVector(X=0, Y=0, Z=&Lerp_Float(A=350, B=100, Alpha=$System.ArtifactProximityToCamera)))
   &SolveForcesAndVelocity(Clamp Velocity=false, Limit Acceleration=false)
  ]
  rendererClasses [&NiagaraSpriteRendererProperties]
  renderers [
   {rendererClass=&NiagaraSpriteRendererProperties}
   {rendererClass=&NiagaraSpriteRendererProperties, rendererIndex=1}
  ]
 }
 NE_SpriteSmoke001 {
  bEnabled=false
  emitterUpdateScript [
   &EmitterState(Life Cycle Mode=::NewEnumerator1, Loop Duration=1, UseLoopDelay=false)
   &SpawnRate(
    SpawnRate=&MakeCustomFloatFromBool(Bool=$User.FadeOut, True Float=0, False Float=30)
    Use Spawn Probability=false
   )
  ]
  particleSpawnScript [
   &InitializeParticle(
    Lifetime Max=1
    Lifetime Min=0.8
    Lifetime Mode=::NewEnumerator1
    Sprite Rotation Mode=::NewEnumerator2
    Sprite Size Mode=::NewEnumerator3
    Uniform Sprite Size=1
    UsePositionOffset=false
    Write Material Random=true
   )
   /SetVariables() {isSetParameters=true, module=SetVariables_#23}
   &CameraOffset(Camera Offset Amount=-10)
   &AddVelocity(Velocity=<1 0 0>, Velocity Speed Scale=&RandomRangeFloat(Minimum=1550, Maximum=1850))
  ]
  particleUpdateScript [
   &self:FollowRibbon(
    Alpha:NiagaraFloat=&FloatFromCurve(FloatCurve=DataInterface() {propertyValues {Curve {keys [<at 0 0>,<at 1 0.8>]}}})
    ParticlePerRibbon=20
    RibbonReader=DataInterface() {propertyValues {EmitterBinding {bindingMode=Other, emitterName=NE_FireTrail}}}
   ) {enabled=false, module=FollowRibbon}
   /SetVariables(Particles.SpriteSize=&Vector2DFromFloat(Value=&Multiply_Float(
      A=&Lerp_Float(
       A=80
       B=450
       Alpha=&FloatFromCurve(FloatCurve=DataInterface() {propertyValues {Curve {keys [
            {arriveTangent=0.7052, interpMode=RCIM_Cubic, leaveTangent=0.7052, tangentMode=RCTM_User, time=0, value=0}
            {arriveTangent=0.0991, interpMode=RCIM_Cubic, leaveTangent=0.0991, tangentMode=RCTM_User, time=1, value=1}
           ]}}})
      )
      B=1
     ))) {isSetParameters=true, module=SetVariables_#24}
   &DynamicMaterialParameters(
    Index 0 Param 1=&Add_Float(
     A=&Multiply_Float(A=$Particles.NormalizedAge, B=1)
     B=&RandomRangeFloat(Minimum=0, Maximum=1)
    )
    Index 0 Param 3=10
    Index 1 Param 1=1
    Index 1 Param 3=1
    Index 1 Param 4=0.5
    Index 2 Param 1=1
    Index 2 Param 2=&RandomRangeFloat(Minimum=0, Maximum=1)
    Index 2 Param 3=&FloatFromCurve(
     FloatCurve=DataInterface() {propertyValues {Curve {keys [
         {arriveTangent=0.8418, interpMode=RCIM_Cubic, leaveTangent=0.8418, tangentMode=RCTM_User, time=0, value=0}
         {arriveTangent=1.2208, interpMode=RCIM_Cubic, leaveTangent=1.2208, tangentMode=RCTM_User, time=0.9, value=1}
        ]}}}
     CurveIndex=&Add_Float(A=$Particles.NormalizedAge, B=$System.Timeline.NormalizedTime)
     Scale Curve=5
    )
    Index 2 Param 4=0.7
    Module.Param0WriteEnabled=true
    Module.Param0WriteEnabled1=true
    Module.Param0WriteEnabled2=true
    Module.Param1WriteEnabled=true
    Module.Param1WriteEnabled1=true
    Module.Param1WriteEnabled2=true
    Module.Param2WriteEnabled=true
    Module.Param2WriteEnabled1=true
    Module.Param2WriteEnabled2=true
    Module.Param3WriteEnabled=true
    Module.Param3WriteEnabled1=true
    Module.Param3WriteEnabled2=true
    Write Parameter Index 1=true
    Write Parameter Index 2=true
   )
   &ParticleState()
   &ScaleColor(
    Scale Alpha=&Multiply_Float(
     A=&FloatFromCurve(
      FloatCurve=DataInterface() {propertyValues {Curve {keys [
          {arriveTangent=15.5597, interpMode=RCIM_Cubic, leaveTangent=15.5597, tangentMode=RCTM_User, time=0.1, value=0}
          {arriveTangent=0.3912, interpMode=RCIM_Cubic, tangentMode=RCTM_Break, time=0.2, value=1}
          {arriveTangent=-0.1085, interpMode=RCIM_Cubic, leaveTangent=-0.1085, tangentMode=RCTM_User, time=0.75, value=1}
          <at 1 0>
         ]}}}
      CurveIndex=&Add_Float(A=$Particles.NormalizedAge, B=$System.Timeline.NormalizedTime)
     )
     B=1
    )
    Scale RGB=&MakeVectorFromLinearColorRGB(Color=&ColorFromCurve(ColorCurve=DataInterface() {propertyValues {
        AlphaCurve {keys []}
        BlueCurve {keys [<at 0.0774 0.0625>,<at 0.3614 0>,<at 0.6427 0 RCIM_Cubic>,<at 0.9972 0 RCIM_Cubic>]}
        GreenCurve {keys [
          <at 0.0774 0.3826>
          <at 0.3614 0.0321>
          {arriveTangent=-0.0437, interpMode=RCIM_Cubic, leaveTangent=-0.0437, time=0.6427, value=0.0226}
          <at 0.9972 0.0044 RCIM_Cubic>
         ]}
        RedCurve {keys [
          <at 0.0774 2>
          <at 0.3614 0.2>
          {arriveTangent=-0.2517, interpMode=RCIM_Cubic, leaveTangent=-0.2517, time=0.6427, value=0.1}
          <at 0.9972 0.04 RCIM_Cubic>
         ]}
       }}))
    ScaleA=true
    ScaleRGB=true
   )
   &AccelerationForce(Acceleration=<0 0 350>)
   &SolveForcesAndVelocity(Clamp Velocity=false, Limit Acceleration=false)
  ]
  rendererClasses [&NiagaraSpriteRendererProperties]
  renderers [
   {rendererClass=&NiagaraSpriteRendererProperties}
   {rendererClass=&NiagaraSpriteRendererProperties, rendererIndex=1}
  ]
 }
 Sparkles {
  emitterUpdateScript [
   &EmitterState(Life Cycle Mode=::NewEnumerator1, Loop Duration=1, UseLoopDelay=false)
   &SpawnRate(
    SpawnRate=&Multiply_Float(A=10, B=&OneMinusFloat(Float=$System.Timeline.NormalizedTime))
    Use Spawn Probability=false
   )
  ]
  particleSpawnScript [
   &InitializeParticle(
    Color=<rgba 10 10 10 1>
    Lifetime=&Multiply_Float(
     A=&RandomRangeFloat(Minimum=0.6, Maximum=1.2)
     B=&Lerp_Float(A=1, B=2, Alpha=$System.ArtifactProximityToCamera)
    )
    Sprite Size=&Multiply_Vector2D_ByFloat(
     Float=&Multiply_Float(
      A=&RandomRangeFloat(Minimum=6, Maximum=8)
      B=&Lerp_Float(A=1, B=3, Alpha=$System.ArtifactProximityToCamera)
     )
     Vector2D=<0.5 1.5>
    )
    Sprite Size Mode=::NewEnumerator4
    UsePositionOffset=false
    Write Material Random=true
   ) {module=InitializeParticle001}
   &AddVelocity(
    Cone Angle=40
    Module.UseConeFalloff=false
    Module.UseVelDistribution=false
    Override Seed=false
    Velocity Mode=::NewEnumerator2
    Velocity Speed=&RandomRangeFloat(Minimum=650, Maximum=1000)
   ) {enabled=false}
   &ShapeLocation(
    MaskSpawnGroup=false
    Override Seed=false
    Sphere Radius=&Lerp_Float(A=50, B=100, Alpha=$System.ArtifactProximityToCamera)
    Sphere Surface Distribution=1
    UseOwnerScale=true
   )
   &AddVelocity(
    Module.Use Falloff Distance=false
    Origin Offset=&MakeVector(X=&Lerp_Float(A=-200, B=-300, Alpha=$System.ArtifactProximityToCamera), Y=0, Z=0)
    Override Seed=false
    Velocity Mode=::NewEnumerator1
    Velocity Speed=&Multiply_Float(
     A=&RandomRangeFloat(Minimum=650, Maximum=1000)
     B=&Lerp_Float(A=1, B=2, Alpha=$System.ArtifactProximityToCamera)
    )
   ) {module=AddVelocity001}
  ]
  particleUpdateScript [
   &ParticleState()
   &ScaleSpriteSize(Scale Factor=&Vector2DFromFloat(Value=&FloatFromCurve(FloatCurve=DataInterface() {propertyValues {Curve {keys [
          {interpMode=RCIM_Cubic, tangentMode=RCTM_User, time=-0.0046, value=0}
          {interpMode=RCIM_Cubic, tangentMode=RCTM_User, time=0.1746, value=0.9952}
          {interpMode=RCIM_Cubic, tangentMode=RCTM_User, time=1.0011, value=0.002}
         ]}}})))
   &Color(
    Color=&ColorFromCurve(ColorCurve=DataInterface() {propertyValues {
       AlphaCurve {keys [<at 0 1 RCIM_Cubic>,<at 1 1 RCIM_Cubic>]}
       BlueCurve {keys [<at 0 0 RCIM_Cubic>,<at 0.1526 0>,<at 0.7427 0 RCIM_Cubic>,<at 1 0>]}
       GreenCurve {keys [
         <at 0 4.0263 RCIM_Cubic>
         <at 0.1526 0.4158>
         {arriveTangent=-0.4907, interpMode=RCIM_Cubic, leaveTangent=-0.4907, time=0.7427, value=0.0752}
         <at 1 0>
        ]}
       RedCurve {keys [<at 0 10 RCIM_Cubic>, <at 0.1526 3>, {arriveTangent=-3.1864, interpMode=RCIM_Cubic, leaveTangent=-3.1864, time=0.7427, value=1}, <at 1 0.3>]}
      }})
    Scale Alpha=&FloatFromCurve(FloatCurve=DataInterface() {propertyValues {Curve {keys [<at -0.0092 0.0144>,<at 0.0967 0.9767>,<at 1.0011 0.002>]}}})
   )
   &DynamicMaterialParameters(
    Index 0 Param 1=1
    Index 0 Param 2=0.5
    Index 0 Param 3=5
    Module.Index0Param0Write=true
    Module.Index0Param1Write=true
    Module.Index0Param2Write=true
    Module.Index0Param3Write=true
   ) {module=DynamicMaterialParameters001}
   /SetVariables(Particles.GlowColor=&MultiplyLinearColor(Linear Color=$Particles.Color, Scale Factor=<1 1 1 0.08>)) {isSetParameters=true, module=SetVariables_#25}
   /SetVariables(Particles.GlowSize=&Multiply_Vector2D(A=$Particles.SpriteSize, B=<2.5 2.5>)) {isSetParameters=true, module=SetVariables_#26}
   &VortexForce(Origin Pull Amount=850, UseOriginOffset=false, Vortex Axis=<1 0 0>, Vortex Force Amount=1500)
   &SolveForcesAndVelocity(Clamp Velocity=false, Limit Acceleration=false)
  ]
  rendererClasses [&NiagaraSpriteRendererProperties]
  renderers [
   {rendererClass=&NiagaraSpriteRendererProperties}
   {rendererClass=&NiagaraSpriteRendererProperties, rendererIndex=1}
  ]
 }
 Sparkles001 {
  bEnabled=false
  emitterUpdateScript [
   &EmitterState(Life Cycle Mode=::NewEnumerator1, Loop Duration=1, UseLoopDelay=false)
   &SpawnRate(
    SpawnRate=&Multiply_Float(A=10, B=&OneMinusFloat(Float=$System.Timeline.NormalizedTime))
    Use Spawn Probability=false
   )
  ]
  particleSpawnScript [
   &InitializeParticle(
    Color=<rgba 10 10 10 1>
    Lifetime Max=1.2
    Lifetime Min=0.6
    Lifetime Mode=::NewEnumerator1
    Sprite Size=&Multiply_Vector2D_ByFloat(Float=&RandomRangeFloat(Minimum=6, Maximum=10), Vector2D=<0.5 1.5>)
    Sprite Size Mode=::NewEnumerator4
    UsePositionOffset=false
    Write Material Random=true
   ) {module=InitializeParticle001}
   &AddVelocity(
    Cone Angle=40
    Module.UseConeFalloff=false
    Module.UseVelDistribution=false
    Override Seed=false
    Velocity Mode=::NewEnumerator2
    Velocity Speed=&RandomRangeFloat(Minimum=650, Maximum=1000)
   ) {enabled=false}
   &ShapeLocation(MaskSpawnGroup=false, Override Seed=false, Sphere Radius=50, Sphere Surface Distribution=1, UseOwnerScale=true)
   &AddVelocity(
    Module.Use Falloff Distance=false
    Origin Offset=<-200 0 0>
    Override Seed=false
    Velocity Mode=::NewEnumerator1
    Velocity Speed=&RandomRangeFloat(Minimum=650, Maximum=1000)
   ) {module=AddVelocity001}
  ]
  particleUpdateScript [
   &ParticleState()
   &ScaleSpriteSize(Scale Factor=&Vector2DFromFloat(Value=&FloatFromCurve(FloatCurve=DataInterface() {propertyValues {Curve {keys [<at -0.0046 0>,<at 0.1746 0.9952>,<at 1.0011 0.002>]}}})))
   &Color(
    Color=&ColorFromCurve(ColorCurve=DataInterface() {propertyValues {
       AlphaCurve {keys [<at 0 1 RCIM_Cubic>,<at 1 1 RCIM_Cubic>]}
       BlueCurve {keys [<at 0 0 RCIM_Cubic>,<at 0.1526 0>,<at 0.7427 0 RCIM_Cubic>,<at 1 0>]}
       GreenCurve {keys [
         <at 0 4.0263 RCIM_Cubic>
         <at 0.1526 0.4158>
         {arriveTangent=-0.4907, interpMode=RCIM_Cubic, leaveTangent=-0.4907, time=0.7427, value=0.0752}
         <at 1 0>
        ]}
       RedCurve {keys [<at 0 10 RCIM_Cubic>, <at 0.1526 3>, {arriveTangent=-3.1864, interpMode=RCIM_Cubic, leaveTangent=-3.1864, time=0.7427, value=1}, <at 1 0.3>]}
      }})
    Scale Alpha=&FloatFromCurve(FloatCurve=DataInterface() {propertyValues {Curve {keys [<at -0.0092 0.0144>,<at 0.0967 0.9767>,<at 1.0011 0.002>]}}})
   )
   &DynamicMaterialParameters(
    Index 0 Param 1=1
    Index 0 Param 2=0.5
    Index 0 Param 3=5
    Module.Index0Param0Write=true
    Module.Index0Param1Write=true
    Module.Index0Param2Write=true
    Module.Index0Param3Write=true
   ) {module=DynamicMaterialParameters001}
   /SetVariables(Particles.GlowColor=&MultiplyLinearColor(Linear Color=$Particles.Color, Scale Factor=<1 1 1 0.08>)) {isSetParameters=true, module=SetVariables_#25}
   /SetVariables(Particles.GlowSize=&Multiply_Vector2D(A=$Particles.SpriteSize, B=<2.5 2.5>)) {isSetParameters=true, module=SetVariables_#26}
   &VortexForce(Origin Pull Amount=850, UseOriginOffset=false, Vortex Axis=<1 0 0>, Vortex Force Amount=1500)
   &SolveForcesAndVelocity(Clamp Velocity=false, Limit Acceleration=false)
  ]
  rendererClasses [&NiagaraSpriteRendererProperties]
  renderers [
   {rendererClass=&NiagaraSpriteRendererProperties}
   {rendererClass=&NiagaraSpriteRendererProperties, rendererIndex=1}
  ]
 }
}
summary {
 systemName=NS_ToadJam_Artifact_FireTrail
 emitters [
  {emitterName=NE_SpriteSmoke, rendererClasses [&NiagaraSpriteRendererProperties]}
  {emitterName=Sparkles, rendererClasses [&NiagaraSpriteRendererProperties]}
  {
   emitterName=Glow_and_Light
   rendererClasses [
    &NiagaraSpriteRendererProperties
    &NiagaraLightRendererProperties
    &NiagaraMeshRendererProperties
   ]
  }
  {emitterName=NE_FlashFlare, rendererClasses [&NiagaraSpriteRendererProperties]}
  {bEnabled=false, emitterName=NE_FlashFlare001, rendererClasses [&NiagaraSpriteRendererProperties]}
  {bEnabled=false, emitterName=NE_FireTrail001, rendererClasses [&NiagaraRibbonRendererProperties]}
  {bEnabled=false, emitterName=NE_SpriteSmoke001, rendererClasses [&NiagaraSpriteRendererProperties]}
  {bEnabled=false, emitterName=Sparkles001, rendererClasses [&NiagaraSpriteRendererProperties]}
  {
   bEnabled=false
   emitterName=Glow_and_Light001
   rendererClasses [
    &NiagaraSpriteRendererProperties
    &NiagaraLightRendererProperties
    &NiagaraMeshRendererProperties
   ]
  }
  {bEnabled=false, emitterName=NE_FlashFlare002, rendererClasses [&NiagaraSpriteRendererProperties]}
  {emitterName=NE_FireTrail, rendererClasses [&NiagaraRibbonRendererProperties]}
 ]
 userVariables [User.FadeOut {type=NiagaraBool}]
}
